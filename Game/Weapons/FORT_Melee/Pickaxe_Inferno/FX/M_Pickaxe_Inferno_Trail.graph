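# Translucent unlit Inferno pickaxe trail material, usable on Niagara ribbons and mesh particles.
# Distortion offsets two panning noises and a highlight, modulating a color gradient blended with camera-vector triplanar nebula and stars.
# Sine-based UV ridge, thickness and edge masks shape opacity, scaled by particle alpha and depth-faded except on mobile.
# Scalar lerps, off by default, switch to a "KneeLense" galaxy look with eroded-cell opacity.
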Material /Game/Weapons/FORT_Melee/Pickaxe_Inferno/FX/M_Pickaxe_Inferno_Trail
graph {
 properties {
  bUsedWithNiagaraMeshParticles=true
  bUsedWithNiagaraRibbons=true
  blendMode=BLEND_Translucent
  pixelDepthOffsetMode=PDOM_Legacy
  shadingModel=MSM_Unlit
  usedShadingModels=Unlit
 }
 nodes [
  n1 = MaterialFunctionCall(Float2=TextureCoordinate() {coordinateIndex=1}) {materialFunction=&BreakOutFloat2Components}
  n2 = Multiply(A=@n1.G) {constB=0.5}
  n3 = ScalarParameter() {defaultValue=1, group=Shape, parameterName=Ridge Top, sortPriority=4}
  n4 = ScalarParameter() {defaultValue=64, group=Shape, parameterName=Ridge Bottom, sortPriority=5}
  n5 = Saturate(None=Power(Base=Sine(None=@n2), Exp=LinearInterpolate(A=@n3, B=@n4, Alpha=@n1.R) {constA=6, constB=2}) {constExponent=4})
  n6 = ScalarParameter() {group=Shape, parameterName=Ridge Mult, sortPriority=6}
  n7 = OneMinus(None=@n1.R)
  n8 = Power(
   Base=Sine(None=Multiply(A=@n1.G) {constB=0.5})
   Exp=LinearInterpolate(
    A=ScalarParameter() {defaultValue=0.5, group=Shape, parameterName=Thickness Top, sortPriority=0}
    B=ScalarParameter() {group=Shape, parameterName=Thickness Bottom, sortPriority=1}
    Alpha=@n7
   ) {constA=6, constB=2}
  ) {constExponent=4}
  n9 = Power(Base=@n8, Exp=ScalarParameter() {defaultValue=1, group=Shape, parameterName=Base Power})
  n10 = ScalarParameter() {defaultValue=1, group=Shape, parameterName=Base Mult}
  n11 = Saturate(None=Add(A=Multiply(A=@n5, B=@n6) {constB=0}, B=Multiply(A=@n9, B=@n10) {constB=0}))
  Add(B=@n11)
  n12 = DynamicParameter() {paramNames [Random,Dissolve,Param3,Param4]}
  n13 = VectorParameter() {group=Noise, parameterName=Noise B UV, sortPriority=6, defaultValue=<rgba 0.75 1 -1.5 -0.1>}
  n14 = AppendVector(A=@n13.R, B=@n13.R)
  n15 = TextureCoordinate()
  n16 = Add(A=AppendVector(A=@n12.Random, B=@n12.Random), B=Panner(Coordinate=Multiply(A=@n15, B=@n14), Speed=AppendVector(A=@n13.B, B=@n13.B)))
  n17 = TextureCoordinate()
  n18 = VectorParameter() {group=Distortion, parameterName=Distortion UV, sortPriority=1, defaultValue=<rgba 1 1 0 0>}
  n19 = AppendVector(A=@n18.R, B=@n18.R)
  n20 = TextureSampleParameter2D(UVs=Panner(Coordinate=Multiply(A=@n17, B=@n19), Speed=AppendVector(A=@n18.B, B=@n18.B))) {
   group=Distortion
   parameterName="Distortion  Tex"
   samplerType=SAMPLERTYPE_Normal
   sortPriority=0
   texture=&T_Weap_Distort_N_02
  }
  n21 = ComponentMask(None=@n17) {g=true}
  n22 = Multiply(
   A=ComponentMask(None=@n20.RGB) {g=true, r=true}
   B=LinearInterpolate(B=OneMinus(None=@n21), Alpha=ScalarParameter() {defaultValue=1, group=Distortion, parameterName=Distortion Gradient Lerp, sortPriority=3}) {constA=1, constAlpha=0}
  )
  n23 = Multiply(A=@n22, B=ScalarParameter() {defaultValue=0.1, group=Distortion, parameterName=Distortion Intensity Noise, sortPriority=2})
  n24 = ChannelMaskParameter(None=TextureSampleParameter2D(UVs=Add(A=@n16, B=@n23)) {group=Noise, parameterName=NoIse B Texture, sortPriority=4, texture=&T_NoisePack_01}.RGB) {group=Noise, maskChannel=Green, parameterName=Noise B Channel, sortPriority=5, defaultValue=<rgba 0 1 0 0>}
  n25 = Multiply(A=@n24.R, B=ScalarParameter() {defaultValue=1, group=Noise, parameterName=NoIse B Intensity, sortPriority=7}) {constB=10}
  n26 = AppendVector(A=@n12.Random, B=@n12.Random)
  n27 = VectorParameter() {group=Noise, parameterName=Noise A UV, sortPriority=2, defaultValue=<rgba 1 0.5 -1.5 0.2>}
  n28 = Panner(Coordinate=Multiply(A=@n15, B=AppendVector(A=@n27.R, B=@n27.R)), Speed=AppendVector(A=@n27.B, B=@n27.B))
  n29 = TextureSampleParameter2D(UVs=Add(A=@n23, B=Add(A=@n26, B=@n28))) {group=Noise, parameterName=NoIse A Texture, sortPriority=0, texture=&T_ShieldCaustics}
  n30 = Multiply(
   A=ChannelMaskParameter(None=@n29.RGB) {group=Noise, parameterName=Noise A Channel, sortPriority=1}.R
   B=ScalarParameter() {defaultValue=1, group=Noise, parameterName=NoIse A Intensity, sortPriority=3}
  ) {constB=10}
  n31 = LinearInterpolate(
   B=LinearInterpolate(
    A=Saturate(None=@n25)
    B=Saturate(None=@n30)
    Alpha=ScalarParameter() {defaultValue=0.35, group=Noise, parameterName=Lerp Noise AB, sortPriority=8}
   )
   Alpha=ScalarParameter() {defaultValue=0.8, group=Noise, parameterName=Lerp Noise, sortPriority=8}
  ) {constA=1}
  n32 = VectorParameter() {group=HighLight, parameterName=HighLight UV, sortPriority=1, defaultValue=<rgba 2 1 2 0>}
  n33 = Multiply(A=TextureCoordinate(), B=AppendVector(A=@n32.R, B=@n32.R))
  n34 = AppendVector(A=@n32.B, B=@n32.B)
  n35 = ScalarParameter() {defaultValue=0.05, group=Distortion, parameterName=Distortion Intensity HighLight, sortPriority=2}
  n36 = TextureSample(UVs=Add(A=Multiply(A=@n22, B=@n35), B=Panner(Coordinate=@n33, Speed=@n34))) {samplerSource=SSM_Wrap_WorldGroupSettings, texture=&/T_Pickaxe_Inferno_Trail_01}
  n37 = VectorParameter() {group=HighLight, parameterName=HighLight Refine, channelNames {a=Lerp, b=Base, g=Mult, r=Power}, defaultValue=<rgba 0.5 30 1 0.5>}
  n38 = ComponentMask(None=TextureCoordinate()) {g=true}
  n39 = Multiply(
   A=Multiply(
    A=MaterialFunctionCall(Texture=@n36.R, RGB=@n37, A=@n37) {materialFunction=&MF_FX_Refine}.Result
    B=VectorParameter() {group=HighLight, parameterName=HighLight Color, defaultValue=<rgba 0.6667 0 1 1>}.RGB
   )
   B=LinearInterpolate(B=OneMinus(None=@n38), Alpha=ScalarParameter() {group=Distortion, parameterName=Distortion Gradient Lerp, sortPriority=3}) {constA=1, constAlpha=0}
  )
  n40 = TextureCoordinate()
  n41 = Power(Base=OneMinus(None=ComponentMask(None=@n40) {r=true}), Exp=ScalarParameter() {defaultValue=8, group=Color, parameterName=Color Power}) {constExponent=4}
  n42 = Multiply(
   A=LinearInterpolate(
    A=VectorParameter() {group=Color, parameterName=Color A, defaultValue=<rgba 1.1146 0.125 2.5 1>}.RGB
    B=VectorParameter() {group=Color, parameterName=Color B, defaultValue=<rgba 0 13.3333 40 1>}.RGB
    Alpha=Saturate(None=@n41)
   )
   B=ScalarParameter() {defaultValue=1, group=Color, parameterName=Color Mult}
  )
  n43 = ScalarParameter() {defaultValue=1, group=Galaxy, parameterName=GalaxyTiling}
  n44 = Constant()
  n45 = ScalarParameter() {defaultValue=6, group=Galaxy, parameterName=GalaxyAxisFade}
  n46 = TextureObjectParameter() {
   group=Galaxy
   parameterName=GalaxyTexture
   texture=/Game/Characters/Player/Male/Medium/Bodies/M_Med_Soldier_04/Skins/Galaxy/Textures/T_FN_Nebula
  }
  n47 = Time()
  n48 = Constant()
  n49 = ScalarParameter() {defaultValue=0.02, group=Galaxy, parameterName=Galaxy Rotation Speed}
  n50 = ScalarParameter() {defaultValue=-0.02, group=Galaxy, parameterName=Galaxy Rotation Speed_1}
  n51 = Multiply(
   A=LinearInterpolate(A=@n47, B=Multiply(A=@n47, B=@n48), Alpha=ScalarParameter() {group=Galaxy, parameterName=Background Rotation})
   B=AppendVector(A=AppendVector(A=@n49, B=@n50), B=ScalarParameter() {defaultValue=0.1, group=Galaxy, parameterName=Galaxy Rotation Speed_2})
  )
  n52 = Multiply(
   A=VectorParameter() {group=Galaxy, parameterName=Galaxy Color, defaultValue=<rgba 1 1 1 1>}.RGB
   B=Desaturation(
    None=MaterialFunctionCall!(CameraVector, Rotation Speed=@n51, Texture=@n46, Tiling=@n43, Offset=@n44, AxisFadeContrast=@n45) {materialFunction=/Game/Weapons/WeaponSkins/MaterialFunction/TriplanarCameraVector_Rotated}.Result
    Fraction=ScalarParameter() {defaultValue=-20, group=Galaxy, parameterName=Galaxy Saturation}
   )
  )
  n53 = Desaturation(None=@n52, Fraction=ScalarParameter() {defaultValue=0.5, group=Galaxy, parameterName=Star Saturation})
  n54 = Time()
  n55 = ScalarParameter() {defaultValue=-0.2, group=Galaxy, parameterName=Small Star Rotation speed}
  n56 = TextureObject() {samplerType=SAMPLERTYPE_Masks, texture=/Game/Characters/Player/Male/Medium/Bodies/M_Med_Soldier_04/Skins/Galaxy/Textures/Celestial_T_Stars}
  n57 = ComponentMask(None=MaterialFunctionCall(
    Rotation Speed=Multiply(A=@n54, B=@n55)
    Texture=@n56
    Tiling=ScalarParameter() {defaultValue=10, group=Galaxy, parameterName=Small Star Tiling}
   ) {materialFunction=&TriplanarCameraVector_Rotated}.Result) {b=true}
  n58 = Multiply(A=@n57, B=@n57)
  n59 = LinearInterpolate(
   A=@n52
   B=Multiply(A=@n53, B=VectorParameter() {group=Galaxy, parameterName=Star Brightness, defaultValue=<rgba 300 300 300 1>}.RGB)
   Alpha=@n58
  )
  n60 = Multiply(A=@n53, B=VectorParameter() {group=Galaxy, parameterName=Star Brightness_1, defaultValue=<rgba 950 950 950 1>}.RGB)
  n61 = TextureObject() {samplerType=SAMPLERTYPE_Masks, texture=&Celestial_T_Stars}
  n62 = ComponentMask(None=MaterialFunctionCall!(
    CameraVector
    Texture=@n61
    Tiling=ScalarParameter() {defaultValue=2, group=Galaxy, parameterName=Bright Star Tiling}
    Offset
    AxisFadeContrast
   ) {materialFunction=/Engine/Functions/Engine_MaterialFunctions01/Texturing/TriplanarCameraVector}.Result) {r=true}
  n63 = TextureSample(UVs=Panner(Coordinate=TextureCoordinate() {uTiling=2, vTiling=2}) {speedX=0.133, speedY=0.5}) {
   samplerSource=SSM_Wrap_WorldGroupSettings
   samplerType=SAMPLERTYPE_Masks
   texture=&Celestial_T_Stars
  }
  n64 = Multiply(A=Multiply(A=@n62, B=@n62), B=@n63.G)
  n65 = LinearInterpolate(
   A=@n42
   B=Multiply(A=LinearInterpolate(A=@n59, B=@n60, Alpha=@n64), B=@n42)
   Alpha=ScalarParameter() {defaultValue=0.5, group=Galaxy, parameterName=Lerp Galaxy, sortPriority=0}
  )
  n66 = Multiply(
   A=@n31
   B=LinearInterpolate(A=@n65, B=Multiply(A=@n39, B=@n65), Alpha=ScalarParameter() {defaultValue=0.5, group=HighLight, parameterName=Lerp Highlight})
  )
  n67 = ParticleColor()
  n68 = NamedRerouteUsage()
  n69 = Constant() {r=1}
  n70 = LinearInterpolate(
   A=Constant3Vector() {constant=<rgba 0.005 0.005 0.005 0>}
   B=Multiply(A=Desaturation(None=@n68, Fraction=@n69)) {constB=0.1}
   Alpha=NamedRerouteUsage()
  )
  n71 = Multiply(
   A=Constant3Vector() {constant=<rgba 1 0.7816 0.0452 1>}
   B=ScalarParameter() {defaultValue=1, group=KneeLensParms, parameterName=KneeLense_StarBrightness}
  )
  n72 = NamedRerouteUsage()
  MaterialFunctionCall(Emissive=LinearInterpolate(
    A=Multiply(A=@n66, B=@n67.RGB)
    B=LinearInterpolate(A=@n70, B=@n71, Alpha=@n72)
    Alpha=ScalarParameter() {group=KneeLensParms, parameterName=UseKneeLenseGalaxy}
   )) {materialFunction=&MF_FX_EmissiveAdjustments}
  n73 = TextureCoordinate()
  n74 = AppendVector(
   A=Multiply(A=ComponentMask(None=@n73) {r=true}, B=ScalarParameter() {defaultValue=0.25, group=KneeLensParms, parameterName=KneeLens_FlameUVMult_U})
   B=Multiply(A=ComponentMask(None=@n73) {g=true})
  )
  n75 = Constant() {r=1.15}
  n76 = ScalarParameter() {parameterName=KneeLens_FlameU_add}
  n77 = ScalarParameter() {parameterName=KneeLens_FlameV_add}
  n78 = Panner(
   Coordinate=Add(
    A=MaterialFunctionCall(UVs=@n74, Texture Scale=@n75) {materialFunction=&ScaleUVsByCenter}.UVs
    B=AppendVector(A=@n76, B=@n77)
   )
   Speed=AppendVector(A=ScalarParameter() {group=KneeLensParms, parameterName=KneeLens_FlameU_PanSpeed}, B=Constant())
  )
  n79 = MaterialFunctionCall(Float2=TextureCoordinate()) {materialFunction=&BreakOutFloat2Components}
  n80 = Multiply(A=@n79.R) {constB=0.5}
  n81 = Multiply(A=@n79.G)
  n82 = ScalarParameter() {defaultValue=-1, group=KneeLensParms, parameterName=KneeLens_PanSpeed}
  n83 = Constant() {r=0.1}
  n84 = TextureSample(UVs=Panner(Coordinate=AppendVector(A=@n80, B=@n81), Speed=AppendVector(A=@n82, B=@n83))) {samplerType=SAMPLERTYPE_LinearColor, texture=/Game/Effects/Fort_Effects/Textures/T_FX_IndieBucket_Contrail_Cells}
  n85 = Power(
   Base=MaterialFunctionCall(Float2=TextureCoordinate()) {materialFunction=&BreakOutFloat2Components}.R
   Exp=ScalarParameter() {defaultValue=8, group=KneeLensParms, parameterName=KneeLens_ErosionPowerRamp_Power}
  ) {constExponent=5}
  n86 = ScalarParameter() {defaultValue=100, group=KneeLensParms, parameterName=KneeLens_ErosionPowerRamp_Mult}
  n87 = MaterialFunctionCall(Float2=TextureCoordinate()) {materialFunction=&BreakOutFloat2Components}
  n88 = Multiply(
   A=Power(Base=OneMinus(None=@n87.R), Exp=ScalarParameter() {defaultValue=1, group=KneeLensParms, parameterName=KneeLens_ErosionMultRamp_Power}) {constExponent=5}
   B=ScalarParameter() {defaultValue=3, group=KneeLensParms, parameterName=KneeLens_ErosionMultRamp_Mult}
  ) {constB=20}
  n89 = Multiply(A=Power(Base=@n84.R)) {constB=2}
  n90 = Sine(None=Multiply(A=@n87.G) {constB=0.5})
  n91 = Power(Base=@n90, Exp=ScalarParameter() {defaultValue=2, group=KneeLensParms, parameterName=KneeLens_ErosionVMultRamp_Power}) {constExponent=5}
  n92 = ScalarParameter() {defaultValue=2, group=KneeLensParms, parameterName=KneeLens_ErosionVMultRamp_Mult}
  n93 = Multiply(A=Add(A=OneMinus(None=@n89), B=Multiply(A=@n91, B=@n92) {constB=20}), B=Multiply(A=Power(Base=@n90) {constExponent=1}))
  n94 = Multiply(A=Power(Base=@n84.RGB, Exp=Multiply(A=@n85, B=@n86) {constB=20}), B=Saturate(None=Multiply(A=@n88, B=@n93)))
  Multiply(A=Multiply(A=TextureSample(UVs=@n78) {texture=/Game/Effects/Niagara/MasterMaterials/SplineThickenTrail/Texture/T_FlameToon_Trail}.G), B=@n94)
  n95 = Power(
   Base=Multiply(A=@n11, B=Multiply(A=@n8, B=@n7))
   Exp=LinearInterpolate(
    A=ScalarParameter() {defaultValue=1, group=Shape, parameterName=Fade Top, sortPriority=2}
    B=ScalarParameter() {defaultValue=1, group=Shape, parameterName=Fade Bottom, sortPriority=3}
    Alpha=@n7
   ) {constA=1, constB=0.5}
  )
  n96 = TextureCoordinate()
  n97 = ComponentMask(None=@n96) {g=true}
  n98 = Multiply(A=Multiply(A=@n97, B=OneMinus(None=@n97)), B=ScalarParameter() {defaultValue=4, group=Shape, parameterName=Edge Fade Power})
  n99 = Multiply(A=Power(Base=ComponentMask(None=@n96) {r=true}) {constExponent=3}) {constB=20}
  n100 = LinearInterpolate(B=Multiply(A=Saturate(None=@n98), B=Saturate(None=@n99)), Alpha=ScalarParameter() {defaultValue=1, group=Shape, parameterName=Use Edge Fade}) {constA=1}
  n101 = Multiply(
   A=LinearInterpolate(A=Multiply(A=@n95, B=@n100), B=@n94, Alpha=ScalarParameter() {group=KneeLensParms, parameterName=UseKneeLenseGalaxy_Flame}) {constA=1}
   B=ScalarParameter() {defaultValue=6, group=Opacity, parameterName=Opacity Multiplier, sortPriority=1}
  ) {constB=2}
  n102 = ScalarParameter() {defaultValue=50, group=Opacity, parameterName=DepthFadeDistance}
  Multiply(A=@n67.A, B=ShadingPathSwitch(Default=DepthFade(Opacity=@n101, FadeDistance=@n102), Mobile=@n101) {desc=Migrated from FeatureLevelSwitch})
  NamedRerouteDeclaration(None=@n31) {name=BackgroundLerp, nodeColor=<rgba 1 0 0.7031 1>}
  NamedRerouteDeclaration(None=@n52) {name=Galaxy, nodeColor=<rgba 0.9453 1 0 1>}
  NamedRerouteDeclaration(None=Add(A=@n58, B=@n64)) {name=Stars, nodeColor=<rgba 0.7344 0 1 1>}
 ]
}
